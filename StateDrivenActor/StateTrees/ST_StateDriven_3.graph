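# StateTree for the StateDrivenActor system: a three-state template (A, B, C) with Success and Failure end states.
# A, B and C each switch on events to either of the other two or to either end state.
# Every child state runs a State Driven Event Transition task, with empty reaction and transition data here.
# Success and Failure add a Blueprint Complete State task, whose completion ends the tree as succeeded or failed.
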StateTree /StateDrivenActor/StateTrees/ST_StateDriven_3
states [
 Root {
  id=#1
  type=STATE
  children [
   #2
   #3
   #4
   #5
   #6
  ]
  tasks []
  transitions []
 }
 A {
  id=#2
  type=STATE
  tasks [{
    id=#7
    description=State Driven Event Transition
    properties {
     iD=#8
     instance {
      _structType=/Script/StateDrivenActorRuntime.StateDrivenEventTransitionStateTreeTaskInstanceData
      instigatorActor=null
      userActor=null
      stateData {devNotes="", instigatorEnterStateReactions [], sourceEnterStateReactions [], transitions []}
     }
     node {
      name=None
      _structType=/Script/StateDrivenActorRuntime.StateDrivenEventTransitionStateTreeTask
      bCanEditConsideredForCompletion=true
      bConsideredForCompletion=true
      bHasTaskCompletionDelegateDispatcher=false
      bTaskEnabled=true
      transitionHandlingPriority=Normal
      bindingsBatch {value=65535}
      executionRuntimeTemplateIndex {value=65535}
      instanceDataHandle {index=65535, source=None, stateHandle {index=65535}}
      instanceTemplateIndex {value=65535}
      outputBindingsBatch {value=65535}
     }
    }
   }]
  transitions [
   {id=#9, trigger=ON_EVENT, state {id=#3, name=B}}
   {id=#10, trigger=ON_EVENT, state {id=#4, name=C}}
   {id=#11, trigger=ON_EVENT, state {id=#5, name=Success}}
   {id=#12, trigger=ON_EVENT, state {id=#6, name=Failure}}
  ]
 }
 B {
  id=#3
  type=STATE
  tasks [{
    id=#13
    description=State Driven Event Transition
    properties {
     iD=#14
     instance {
      _structType=&StateDrivenEventTransitionStateTreeTaskInstanceData
      instigatorActor=null
      userActor=null
      stateData {devNotes="", instigatorEnterStateReactions [], sourceEnterStateReactions [], transitions []}
     }
     node {
      name=None
      _structType=&StateDrivenEventTransitionStateTreeTask
      bCanEditConsideredForCompletion=true
      bConsideredForCompletion=true
      bHasTaskCompletionDelegateDispatcher=false
      bTaskEnabled=true
      transitionHandlingPriority=Normal
      bindingsBatch {value=65535}
      executionRuntimeTemplateIndex {value=65535}
      instanceDataHandle {index=65535, source=None, stateHandle {index=65535}}
      instanceTemplateIndex {value=65535}
      outputBindingsBatch {value=65535}
     }
    }
   }]
  transitions [
   {id=#15, trigger=ON_EVENT, state {id=#2, name=A}}
   {id=#16, trigger=ON_EVENT, state {id=#4, name=C}}
   {id=#17, trigger=ON_EVENT, state {id=#5, name=Success}}
   {id=#18, trigger=ON_EVENT, state {id=#6, name=Failure}}
  ]
 }
 C {
  id=#4
  type=STATE
  tasks [{
    id=#19
    description=State Driven Event Transition
    properties {
     iD=#20
     instance {
      _structType=&StateDrivenEventTransitionStateTreeTaskInstanceData
      instigatorActor=null
      userActor=null
      stateData {devNotes="", instigatorEnterStateReactions [], sourceEnterStateReactions [], transitions []}
     }
     node {
      name=None
      _structType=&StateDrivenEventTransitionStateTreeTask
      bCanEditConsideredForCompletion=true
      bConsideredForCompletion=true
      bHasTaskCompletionDelegateDispatcher=false
      bTaskEnabled=true
      transitionHandlingPriority=Normal
      bindingsBatch {value=65535}
      executionRuntimeTemplateIndex {value=65535}
      instanceDataHandle {index=65535, source=None, stateHandle {index=65535}}
      instanceTemplateIndex {value=65535}
      outputBindingsBatch {value=65535}
     }
    }
   }]
  transitions [
   {id=#21, trigger=ON_EVENT, state {id=#2, name=A}}
   {id=#22, trigger=ON_EVENT, state {id=#3, name=B}}
   {id=#23, trigger=ON_EVENT, state {id=#5, name=Success}}
   {id=#24, trigger=ON_EVENT, state {id=#6, name=Failure}}
  ]
 }
 Success {
  id=#5
  type=STATE
  tasks [
   {
    id=#25
    description=State Driven Event Transition
    properties {
     iD=#26
     instance {
      _structType=&StateDrivenEventTransitionStateTreeTaskInstanceData
      instigatorActor=null
      userActor=null
      stateData {devNotes="", instigatorEnterStateReactions [], sourceEnterStateReactions [], transitions []}
     }
     node {
      name=None
      _structType=&StateDrivenEventTransitionStateTreeTask
      bCanEditConsideredForCompletion=true
      bConsideredForCompletion=true
      bHasTaskCompletionDelegateDispatcher=false
      bTaskEnabled=true
      transitionHandlingPriority=Normal
      bindingsBatch {value=65535}
      executionRuntimeTemplateIndex {value=65535}
      instanceDataHandle {index=65535, source=None, stateHandle {index=65535}}
      instanceTemplateIndex {value=65535}
      outputBindingsBatch {value=65535}
     }
    }
   }
   {
    id=#27
    description=BP SDA STTask Complete State
    properties {
     iD=#28
     instanceObject=&self:StateTreeEditorData_0.StateTreeState_1.BP_SDA_STTask_CompleteState_C_0
     node {
      name=BP SDA STTask Complete State
      _structType=&StateTreeBlueprintTaskWrapper
      taskClass=/StateDrivenActor/Tasks/BP_SDA_STTask_CompleteState.BP_SDA_STTask_CompleteState_C
     }
    }
   }
  ]
  transitions [{id=#29, trigger=ON_STATE_COMPLETED, state {id=#30, name=Root, linkType=SUCCEEDED}}]
 }
 Failure {
  id=#6
  type=STATE
  tasks [
   {
    id=#31
    description=State Driven Event Transition
    properties {
     iD=#32
     instance {
      _structType=&StateDrivenEventTransitionStateTreeTaskInstanceData
      instigatorActor=null
      userActor=null
      stateData {devNotes="", instigatorEnterStateReactions [], sourceEnterStateReactions [], transitions []}
     }
     node {
      name=None
      _structType=&StateDrivenEventTransitionStateTreeTask
      bCanEditConsideredForCompletion=true
      bConsideredForCompletion=true
      bHasTaskCompletionDelegateDispatcher=false
      bTaskEnabled=true
      transitionHandlingPriority=Normal
      bindingsBatch {value=65535}
      executionRuntimeTemplateIndex {value=65535}
      instanceDataHandle {index=65535, source=None, stateHandle {index=65535}}
      instanceTemplateIndex {value=65535}
      outputBindingsBatch {value=65535}
     }
    }
   }
   {
    id=#33
    description=BP SDA STTask Complete State
    properties {
     iD=#34
     instanceObject=&self:StateTreeEditorData_0.StateTreeState_2.BP_SDA_STTask_CompleteState_C_1
     node {
      name=BP SDA STTask Complete State
      _structType=&StateTreeBlueprintTaskWrapper
      taskClass=&BP_SDA_STTask_CompleteState_C
     }
    }
   }
  ]
  transitions [{id=#35, trigger=ON_STATE_COMPLETED, state {id=#30, name=Root, linkType=FAILED}}]
 }
]
